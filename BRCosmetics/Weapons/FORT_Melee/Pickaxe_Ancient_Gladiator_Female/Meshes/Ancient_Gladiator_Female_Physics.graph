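# Physics asset for the female Ancient Gladiator two-handed pickaxe, defining collision bodies and joints for its bandage bones.
# A long capsule covers the handle, and small boxes cover each dyn_bandage bone.
# The a01–a02 chain and the central dyn_bandage body both attach to the handle; the b01–b02 chain and single c, d and e pieces branch from the central body.
# Tight swing and twist limits restrict the bandages to slight angular movement.
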PhysicsAsset /BRCosmetics/Weapons/FORT_Melee/Pickaxe_Ancient_Gladiator_Female/Meshes/Ancient_Gladiator_Female_Physics
bodies {
 Melee_TwoHanded_Handle [{shapeType=Capsule, length=107.5976, radius=1.257, center=<0.3148 -1.0984 27.4909>, rotation=<rot -0.043 -90 -0.0097>}]
 dyn_bandage [{shapeType=Box, extentX=4.4147, extentY=4.342, extentZ=3.08}]
 dyn_bandage_a01 [{shapeType=Box, extentX=1.0209, extentY=1.2885, extentZ=1.5015}]
 dyn_bandage_a02 [{shapeType=Box, extentX=0.7157, extentY=3.1713, extentZ=5.9355, center=<0 0 -0.4211>}]
 dyn_bandage_b01 [{shapeType=Box, extentX=0.5558, extentY=3.1713, extentZ=5.6514, center=<0 0 -0.2991>}]
 dyn_bandage_b02 [{shapeType=Box, extentX=0.5558, extentY=3.1713, extentZ=4.258, center=<0 0 1.001>}]
 dyn_bandage_c01 [{shapeType=Box, extentX=0.5558, extentY=3.1713, extentZ=5.9355}]
 dyn_bandage_d01 [{shapeType=Box, extentX=2.3732, extentY=1.0467, extentZ=5.9355}]
 dyn_bandage_e01 [{shapeType=Box, extentX=0.5558, extentY=3.1713, extentZ=5.9355}]
}
constraints [
 {bone1Name=dyn_bandage_a01, bone2Name=Melee_TwoHanded_Handle, swing1LimitDegrees=1, swing2LimitDegrees=1, twistLimitDegrees=1}
 {bone1Name=dyn_bandage_a02, bone2Name=dyn_bandage_a01, swing1LimitDegrees=5, swing2LimitDegrees=10, twistLimitDegrees=5}
 {bone1Name=dyn_bandage, bone2Name=Melee_TwoHanded_Handle, swing1LimitDegrees=1, swing2LimitDegrees=5, twistLimitDegrees=1}
 {bone1Name=dyn_bandage_b01, bone2Name=dyn_bandage, swing1LimitDegrees=0.1, swing2LimitDegrees=1, twistLimitDegrees=0.1}
 {bone1Name=dyn_bandage_b02, bone2Name=dyn_bandage_b01, swing1LimitDegrees=1, swing2LimitDegrees=2, twistLimitDegrees=1}
 {bone1Name=dyn_bandage_c01, bone2Name=dyn_bandage, swing1LimitDegrees=1, swing2LimitDegrees=5, twistLimitDegrees=1}
 {bone1Name=dyn_bandage_d01, bone2Name=dyn_bandage, swing1LimitDegrees=1, swing2LimitDegrees=5, twistLimitDegrees=1}
 {bone1Name=dyn_bandage_e01, bone2Name=dyn_bandage, swing1LimitDegrees=1, swing2LimitDegrees=5, twistLimitDegrees=1}
]
